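# Physics asset for the ScotchPeek male medium player body cosmetic, set up for a dynamic robe and sword.
# Capsules on pelvis, spine, legs and feet accompany seven five-segment robe chains, anchored by base bones to spine_03, plus a ground collision box.
# Robe chains use tight twist and varied swing limits, with cross-links joining back panels and angularly locking front to side panels mid-chain.
# The sword capsule is constrained to spine_01 with free rotation.
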PhysicsAsset /BRCosmetics/Characters/Player/Male/Medium/Bodies/M_MED_ScotchPeek/Meshes/M_MED_ScotchPeek_Physics
bodies {
 ball_l [{shapeType=Capsule, length=11.6907, radius=8.7492, shapeName=ball_l_capsule, center=<-8.6611 0 -0.7723>, rotation=<rot 90 0 0>}]
 ball_r [{shapeType=Capsule, length=13.4149, radius=8.6154, shapeName=ball_r_capsule, center=<9.0421 0 -0.7723>, rotation=<rot 90 0 0>}]
 base_robeBack_l [{shapeType=Capsule, length=6.626, radius=5.3507, shapeName=base_robeBack_l_capsule}]
 base_robeBack_m [{shapeType=Capsule, length=3.1463, radius=1.5091, shapeName=base_robeBack_m_capsule}]
 base_robeBack_r [{shapeType=Capsule, length=0.505, radius=0.505, shapeName=base_robeBack_r_capsule}]
 base_robeFront_l [{shapeType=Capsule, length=0.505, radius=0.505, shapeName=base_robeFront_l_capsule}]
 base_robeFront_r [{shapeType=Capsule, length=0.505, radius=0.505, shapeName=base_robeFront_r_capsule}]
 base_robeSide_l [{shapeType=Capsule, length=5.3425, radius=1.6353, shapeName=base_robeSide_l_capsule}]
 base_robeSide_r [{shapeType=Capsule, length=0.505, radius=0.505, shapeName=base_robeSide_r_capsule}]
 calf_l [{shapeType=Capsule, length=25.8128, radius=14.6924, shapeName=calf_l_capsule, center=<-20.4519 -1.5 -0.0103>, rotation=<rot 87.7811 0 0.0001>}]
 calf_r [{
   shapeType=Capsule
   length=31.6575
   radius=15.2383
   shapeName=calf_r_capsule
   center=<20.8324 -0.8392 -0.2141>
   rotation=<rot 88.3894 -15.3964 -11.4128>
  }]
 dyn_Sword [{
   shapeType=Capsule
   length=91.0725
   radius=6.6282
   shapeName=dyn_Sword_capsule
   center=<18.9517 9.8842 -0.7591>
   rotation=<rot 0 113.2843 -89.5322>
  }]
 dyn_robeBaack_02_l [{shapeType=Capsule, length=14, radius=11, shapeName=dyn_robeBaack_02_l_capsule, center=<0 -8.0722 0>}]
 dyn_robeBaack_02_r [{shapeType=Capsule, length=14, radius=11, shapeName=dyn_robeBaack_02_r_capsule, center=<0 7.9493 0>}]
 dyn_robeBack_01_l [{shapeType=Capsule, length=14, radius=11, shapeName=dyn_robeBack_01_l_capsule, center=<0 -0.9541 0>}]
 dyn_robeBack_01_m [{shapeType=Capsule, length=14, radius=11, shapeName=dyn_robeBack_01_m_capsule, center=<0 -9.3891 -0.7866>}]
 dyn_robeBack_01_r [{shapeType=Capsule, length=14, radius=11, shapeName=dyn_robeBack_01_r_capsule}]
 dyn_robeBack_02_m [{shapeType=Capsule, length=25.6689, radius=11, shapeName=dyn_robeBack_02_m_capsule, center=<0 -8.1424 0>}]
 dyn_robeBack_03_l [{shapeType=Capsule, length=14, radius=11, shapeName=dyn_robeBack_03_l_capsule, center=<0 -9.3224 0>}]
 dyn_robeBack_03_m [{shapeType=Capsule, length=25.6689, radius=11, shapeName=dyn_robeBack_03_m_capsule, center=<0 -8.4521 0>}]
 dyn_robeBack_03_r [{shapeType=Capsule, length=14, radius=11, shapeName=dyn_robeBack_03_r_capsule, center=<0 9.6894 0>}]
 dyn_robeBack_04_l [{shapeType=Capsule, length=14, radius=11, shapeName=dyn_robeBack_04_l_capsule, center=<0 -9.7741 0>}]
 dyn_robeBack_04_m [{shapeType=Capsule, length=25.6689, radius=11, shapeName=dyn_robeBack_04_m_capsule, center=<0 -4.8983 0>}]
 dyn_robeBack_04_r [{shapeType=Capsule, length=14, radius=11, shapeName=dyn_robeBack_04_r_capsule, center=<0 9.5181 0>}]
 dyn_robeBack_05_l [{shapeType=Capsule, length=24.9287, radius=11, shapeName=dyn_robeBack_05_l_capsule, center=<0 -8.4381 0>}]
 dyn_robeBack_05_m [{shapeType=Capsule, length=25.6689, radius=11, shapeName=dyn_robeBack_05_m_capsule, center=<0 -3.2761 0>}]
 dyn_robeBack_05_r [{shapeType=Capsule, length=24.4101, radius=11, shapeName=dyn_robeBack_05_r_capsule, center=<0 8.9009 0>}]
 dyn_robeFront_01_l [{shapeType=Capsule, length=14, radius=11, shapeName=dyn_robeFront_01_l_capsule, center=<0 -7.9919 0>}]
 dyn_robeFront_01_r [{shapeType=Capsule, length=14, radius=11, shapeName=dyn_robeFront_01_r_capsule, center=<0 8.1416 0>}]
 dyn_robeFront_02_l [{shapeType=Capsule, length=14, radius=11, shapeName=dyn_robeFront_02_l_capsule, center=<0 -10.4546 0.6173>, rotation=<rot 0 0 19.5665>}]
 dyn_robeFront_02_r [{shapeType=Capsule, length=14, radius=11, shapeName=dyn_robeFront_02_r_capsule, center=<0 6.0251 0>}]
 dyn_robeFront_03_l [{shapeType=Capsule, length=14, radius=11, shapeName=dyn_robeFront_03_l_capsule, center=<0 -9.5755 0.5987>}]
 dyn_robeFront_03_r [{shapeType=Capsule, length=14, radius=11, shapeName=dyn_robeFront_03_r_capsule, center=<0 7.2745 0>}]
 dyn_robeFront_04_l [{shapeType=Capsule, length=14, radius=11, shapeName=dyn_robeFront_04_l_capsule, center=<0 -9.4102 0.2728>}]
 dyn_robeFront_04_r [{shapeType=Capsule, length=14, radius=11, shapeName=dyn_robeFront_04_r_capsule, center=<0 7.3111 0>}]
 dyn_robeFront_05_l [{shapeType=Capsule, length=14, radius=11, shapeName=dyn_robeFront_05_l_capsule, center=<0 -5.3557 0>}]
 dyn_robeFront_05_r [{shapeType=Capsule, length=14, radius=11, shapeName=dyn_robeFront_05_r_capsule, center=<0 7.434 0>}]
 dyn_robeSide_01_l [{shapeType=Capsule, length=14, radius=11, shapeName=dyn_robeSide_01_l_capsule, center=<0 -7.9987 0>}]
 dyn_robeSide_01_r [{shapeType=Capsule, length=22.9433, radius=12.2504, shapeName=dyn_robeSide_01_r_capsule, center=<0 8.9687 0>}]
 dyn_robeSide_02_l [{shapeType=Capsule, length=28.6989, radius=12.8473, shapeName=dyn_robeSide_02_l_capsule, center=<0 -7.5402 0>}]
 dyn_robeSide_02_r [{shapeType=Capsule, length=22.9433, radius=12.2504, shapeName=dyn_robeSide_02_r_capsule, center=<0 6.9687 0>}]
 dyn_robeSide_03_l [{shapeType=Capsule, length=28.6989, radius=12.8473, shapeName=dyn_robeSide_03_l_capsule, center=<0 -9.0896 0>}]
 dyn_robeSide_03_r [{shapeType=Capsule, length=22.9433, radius=12.2504, shapeName=dyn_robeSide_03_r_capsule, center=<0 9.4715 0>}]
 dyn_robeSide_04_l [{shapeType=Capsule, length=28.6989, radius=12.8473, shapeName=dyn_robeSide_04_l_capsule, center=<0 -8.0222 0>}]
 dyn_robeSide_04_r [{shapeType=Capsule, length=22.9433, radius=12.2504, shapeName=dyn_robeSide_04_r_capsule, center=<0 8.3625 0>}]
 dyn_robeSide_05_l [{shapeType=Capsule, length=28.6989, radius=12.8473, shapeName=dyn_robeSide_05_l_capsule, center=<0 -7.3893 0>}]
 dyn_robeSide_05_r [{shapeType=Capsule, length=22.9433, radius=12.2504, shapeName=dyn_robeSide_05_r_capsule, center=<0 8.1448 0>}]
 foot_l [{shapeType=Capsule, length=27.2061, radius=8.9394, shapeName=foot_l_capsule, center=<12.38 0 0>, rotation=<rot 90 0 0>}]
 foot_r [{shapeType=Capsule, length=27.2363, radius=9.8441, shapeName=foot_r_capsule, center=<-12.9637 0 0>, rotation=<rot 90 0 0>}]
 ground_collision [{shapeType=Box, extentX=296.6484, extentY=264.8772, extentZ=10.6992, shapeName=ground_collision_box, rotation=<rot 0 -90 0>}]
 pelvis [{shapeType=Capsule, length=7.1802, radius=12.0642, shapeName=pelvis_capsule, center=<0 2.1433 0>}]
 spine_01 [{shapeType=Capsule, length=6.5513, radius=12.3861, shapeName=spine_01_capsule, center=<0 0.5169 0>}]
 spine_02 [{shapeType=Capsule, length=8.0249, radius=11.3932, shapeName=spine_02_capsule, center=<0 2.3681 0>}]
 spine_03 [{shapeType=Capsule, length=3.9082, radius=12.8437, shapeName=spine_03_capsule, center=<0 1.8557 0>}]
 thigh_l [{
   shapeType=Capsule
   length=33.0762
   radius=13.9449
   shapeName=thigh_l_capsule
   center=<-22.2992 -1.7671 1.4923>
   rotation=<rot 85.1744 -180 -180>
  }]
 thigh_r [{shapeType=Capsule, length=27.8652, radius=14.8994, shapeName=thigh_r_capsule, center=<19.6861 0.981 -1.3504>, rotation=<rot 84.3859 180 180>}]
}
constraints [
 {bone1Name=dyn_robeBack_01_m, bone2Name=base_robeBack_m, swing1LimitDegrees=30, swing2LimitDegrees=20, twistLimitDegrees=1}
 {bone1Name=dyn_robeBack_02_m, bone2Name=dyn_robeBack_01_m, swing1LimitDegrees=30, swing2LimitDegrees=20, twistLimitDegrees=1}
 {bone1Name=dyn_robeBack_03_m, bone2Name=dyn_robeBack_02_m, swing1LimitDegrees=30, swing2LimitDegrees=20, twistLimitDegrees=1}
 {bone1Name=dyn_robeBack_04_m, bone2Name=dyn_robeBack_03_m, swing1LimitDegrees=30, swing2LimitDegrees=20, twistLimitDegrees=1}
 {bone1Name=dyn_robeBack_05_m, bone2Name=dyn_robeBack_04_m, swing1LimitDegrees=30, swing2LimitDegrees=20, twistLimitDegrees=1}
 {bone1Name=dyn_robeFront_01_l, bone2Name=base_robeFront_l, swing1LimitDegrees=25, swing2LimitDegrees=10, twistLimitDegrees=1}
 {bone1Name=dyn_robeFront_02_l, bone2Name=dyn_robeFront_01_l, swing1LimitDegrees=45, swing2LimitDegrees=10, twistLimitDegrees=1}
 {bone1Name=dyn_robeFront_03_l, bone2Name=dyn_robeFront_02_l, swing1LimitDegrees=45, swing2LimitDegrees=20, twistLimitDegrees=1}
 {bone1Name=dyn_robeFront_04_l, bone2Name=dyn_robeFront_03_l, swing1LimitDegrees=45, swing2LimitDegrees=20, twistLimitDegrees=1}
 {bone1Name=dyn_robeFront_05_l, bone2Name=dyn_robeFront_04_l, swing1LimitDegrees=45, swing2LimitDegrees=20, twistLimitDegrees=1}
 {bone1Name=dyn_robeFront_01_r, bone2Name=base_robeFront_r, swing1LimitDegrees=15, swing2LimitDegrees=10, twistLimitDegrees=1}
 {bone1Name=dyn_robeFront_02_r, bone2Name=dyn_robeFront_01_r, swing1LimitDegrees=15, swing2LimitDegrees=10, twistLimitDegrees=1}
 {bone1Name=dyn_robeFront_03_r, bone2Name=dyn_robeFront_02_r, swing1LimitDegrees=15, swing2LimitDegrees=10, twistLimitDegrees=1}
 {bone1Name=dyn_robeFront_04_r, bone2Name=dyn_robeFront_03_r, swing1LimitDegrees=15, swing2LimitDegrees=10, twistLimitDegrees=1}
 {bone1Name=dyn_robeFront_05_r, bone2Name=dyn_robeFront_04_r, swing1LimitDegrees=15, swing2LimitDegrees=10, twistLimitDegrees=1}
 {bone1Name=dyn_robeSide_01_l, bone2Name=base_robeSide_l, swing1LimitDegrees=15, swing2LimitDegrees=5, twistLimitDegrees=1}
 {bone1Name=dyn_robeSide_02_l, bone2Name=dyn_robeSide_01_l, swing1LimitDegrees=15, swing2LimitDegrees=5, twistLimitDegrees=1}
 {bone1Name=dyn_robeSide_03_l, bone2Name=dyn_robeSide_02_l, swing1LimitDegrees=35, swing2LimitDegrees=10, twistLimitDegrees=1}
 {bone1Name=dyn_robeSide_04_l, bone2Name=dyn_robeSide_03_l, swing1LimitDegrees=15, swing2LimitDegrees=10, twistLimitDegrees=1}
 {bone1Name=dyn_robeSide_05_l, bone2Name=dyn_robeSide_04_l, swing1LimitDegrees=15, swing2LimitDegrees=10, twistLimitDegrees=1}
 {bone1Name=dyn_robeSide_01_r, bone2Name=base_robeSide_r, swing1LimitDegrees=15, swing2LimitDegrees=10, twistLimitDegrees=1}
 {bone1Name=dyn_robeSide_02_r, bone2Name=dyn_robeSide_01_r, swing1LimitDegrees=15, swing2LimitDegrees=10, twistLimitDegrees=1}
 {bone1Name=dyn_robeSide_03_r, bone2Name=dyn_robeSide_02_r, swing1LimitDegrees=15, swing2LimitDegrees=10, twistLimitDegrees=1}
 {bone1Name=dyn_robeSide_04_r, bone2Name=dyn_robeSide_03_r, swing1LimitDegrees=15, swing2LimitDegrees=10, twistLimitDegrees=1}
 {bone1Name=dyn_robeSide_05_r, bone2Name=dyn_robeSide_04_r, swing1LimitDegrees=15, swing2LimitDegrees=10, twistLimitDegrees=1}
 {
  bone1Name=base_robeSide_l
  bone2Name=spine_03
  swing1LimitDegrees=45
  swing1Motion=Free
  swing2LimitDegrees=45
  swing2Motion=Free
  twistLimitDegrees=45
  twistMotion=Free
 }
 {
  bone1Name=base_robeBack_l
  bone2Name=spine_03
  swing1LimitDegrees=45
  swing1Motion=Free
  swing2LimitDegrees=45
  swing2Motion=Free
  twistLimitDegrees=45
  twistMotion=Free
 }
 {
  bone1Name=base_robeBack_m
  bone2Name=spine_03
  swing1LimitDegrees=45
  swing1Motion=Free
  swing2LimitDegrees=45
  swing2Motion=Free
  twistLimitDegrees=45
  twistMotion=Free
 }
 {
  bone1Name=base_robeFront_l
  bone2Name=spine_03
  swing1LimitDegrees=45
  swing1Motion=Free
  swing2LimitDegrees=45
  swing2Motion=Free
  twistLimitDegrees=45
  twistMotion=Free
 }
 {
  bone1Name=base_robeSide_r
  bone2Name=spine_03
  swing1LimitDegrees=45
  swing1Motion=Free
  swing2LimitDegrees=45
  swing2Motion=Free
  twistLimitDegrees=45
  twistMotion=Free
 }
 {
  bone1Name=base_robeBack_r
  bone2Name=spine_03
  swing1LimitDegrees=45
  swing1Motion=Free
  swing2LimitDegrees=45
  swing2Motion=Free
  twistLimitDegrees=45
  twistMotion=Free
 }
 {
  bone1Name=base_robeFront_r
  bone2Name=spine_03
  swing1LimitDegrees=45
  swing1Motion=Free
  swing2LimitDegrees=45
  swing2Motion=Free
  twistLimitDegrees=45
  twistMotion=Free
 }
 {
  bone1Name=spine_01
  bone2Name=pelvis
  swing1LimitDegrees=45
  swing1Motion=Free
  swing2LimitDegrees=45
  swing2Motion=Free
  twistLimitDegrees=45
  twistMotion=Free
 }
 {
  bone1Name=spine_02
  bone2Name=spine_01
  swing1LimitDegrees=45
  swing1Motion=Free
  swing2LimitDegrees=45
  swing2Motion=Free
  twistLimitDegrees=45
  twistMotion=Free
 }
 {
  bone1Name=spine_03
  bone2Name=spine_02
  swing1LimitDegrees=45
  swing1Motion=Free
  swing2LimitDegrees=45
  swing2Motion=Free
  twistLimitDegrees=45
  twistMotion=Free
 }
 {
  bone1Name=thigh_l
  bone2Name=pelvis
  swing1LimitDegrees=45
  swing1Motion=Free
  swing2LimitDegrees=45
  swing2Motion=Free
  twistLimitDegrees=45
  twistMotion=Free
 }
 {
  bone1Name=calf_l
  bone2Name=thigh_l
  swing1LimitDegrees=45
  swing1Motion=Free
  swing2LimitDegrees=45
  swing2Motion=Free
  twistLimitDegrees=45
  twistMotion=Free
 }
 {
  bone1Name=foot_l
  bone2Name=calf_l
  swing1LimitDegrees=45
  swing1Motion=Free
  swing2LimitDegrees=45
  swing2Motion=Free
  twistLimitDegrees=45
  twistMotion=Free
 }
 {
  bone1Name=ball_l
  bone2Name=foot_l
  swing1LimitDegrees=45
  swing1Motion=Free
  swing2LimitDegrees=45
  swing2Motion=Free
  twistLimitDegrees=45
  twistMotion=Free
 }
 {
  bone1Name=thigh_r
  bone2Name=pelvis
  swing1LimitDegrees=45
  swing1Motion=Free
  swing2LimitDegrees=45
  swing2Motion=Free
  twistLimitDegrees=45
  twistMotion=Free
 }
 {
  bone1Name=calf_r
  bone2Name=thigh_r
  swing1LimitDegrees=45
  swing1Motion=Free
  swing2LimitDegrees=45
  swing2Motion=Free
  twistLimitDegrees=45
  twistMotion=Free
 }
 {
  bone1Name=foot_r
  bone2Name=calf_r
  swing1LimitDegrees=45
  swing1Motion=Free
  swing2LimitDegrees=45
  swing2Motion=Free
  twistLimitDegrees=45
  twistMotion=Free
 }
 {
  bone1Name=ball_r
  bone2Name=foot_r
  swing1LimitDegrees=45
  swing1Motion=Free
  swing2LimitDegrees=45
  swing2Motion=Free
  twistLimitDegrees=45
  twistMotion=Free
 }
 {bone1Name=dyn_robeBack_01_r, bone2Name=base_robeBack_r, swing1LimitDegrees=15, swing2LimitDegrees=10, twistLimitDegrees=1}
 {bone1Name=dyn_robeBaack_02_r, bone2Name=dyn_robeBack_01_r, swing1LimitDegrees=15, swing2LimitDegrees=10, twistLimitDegrees=1}
 {bone1Name=dyn_robeBack_03_r, bone2Name=dyn_robeBaack_02_r, swing1LimitDegrees=30, swing2LimitDegrees=10, twistLimitDegrees=1}
 {bone1Name=dyn_robeBack_04_r, bone2Name=dyn_robeBack_03_r, swing1LimitDegrees=30, swing2LimitDegrees=10, twistLimitDegrees=1}
 {bone1Name=dyn_robeBack_05_r, bone2Name=dyn_robeBack_04_r, swing1LimitDegrees=30, swing2LimitDegrees=10, twistLimitDegrees=1}
 {bone1Name=dyn_robeBack_01_l, bone2Name=base_robeBack_l, swing1LimitDegrees=15, swing2LimitDegrees=10, twistLimitDegrees=1}
 {bone1Name=dyn_robeBaack_02_l, bone2Name=dyn_robeBack_01_l, swing1LimitDegrees=16, swing2LimitDegrees=10, twistLimitDegrees=1}
 {bone1Name=dyn_robeBack_03_l, bone2Name=dyn_robeBaack_02_l, swing1LimitDegrees=20, swing2LimitDegrees=10, twistLimitDegrees=1}
 {bone1Name=dyn_robeBack_04_l, bone2Name=dyn_robeBack_03_l, swing1LimitDegrees=30, swing2LimitDegrees=10, twistLimitDegrees=1}
 {bone1Name=dyn_robeBack_05_l, bone2Name=dyn_robeBack_04_l, swing1LimitDegrees=30, swing2LimitDegrees=10, twistLimitDegrees=1}
 {
  bone1Name=dyn_robeBaack_02_r
  bone2Name=dyn_robeBack_02_m
  swing1LimitDegrees=45
  swing1Motion=Free
  swing2LimitDegrees=45
  swing2Motion=Free
  twistLimitDegrees=45
  twistMotion=Free
 }
 {
  bone1Name=dyn_robeBaack_02_l
  bone2Name=dyn_robeBack_02_m
  swing1LimitDegrees=45
  swing1Motion=Free
  swing2LimitDegrees=45
  swing2Motion=Free
  twistLimitDegrees=45
  twistMotion=Free
 }
 {
  bone1Name=dyn_robeFront_03_l
  bone2Name=dyn_robeSide_03_l
  swing1LimitDegrees=45
  swing1Motion=Locked
  swing2LimitDegrees=45
  swing2Motion=Locked
  twistLimitDegrees=45
  twistMotion=Locked
 }
 {
  bone1Name=dyn_robeFront_03_r
  bone2Name=dyn_robeSide_03_r
  swing1LimitDegrees=45
  swing1Motion=Locked
  swing2LimitDegrees=45
  swing2Motion=Locked
  twistLimitDegrees=45
  twistMotion=Locked
 }
 {
  bone1Name=dyn_Sword
  bone2Name=spine_01
  swing1LimitDegrees=45
  swing1Motion=Free
  swing2LimitDegrees=45
  swing2Motion=Free
  twistLimitDegrees=45
  twistMotion=Free
 }
]
