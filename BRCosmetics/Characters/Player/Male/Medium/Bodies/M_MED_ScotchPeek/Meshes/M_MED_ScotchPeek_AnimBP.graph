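# Post-process style animation Blueprint for the ScotchPeek medium male cosmetic body, adding secondary motion to the attached parent's pose.
# The copied pose runs through a state-driven Control Rig, sleeve and sword-string AnimDynamics, and ground_collision bone transforms, then is cached.
# bClothEnabled blends in RigidBody simulation and enables the cloth controller, which names a NeonFriction coat clothing asset.
# Each update also computes gravity override, running state and interpolated slope-warp values.
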AnimBlueprint /BRCosmetics/Characters/Player/Male/Medium/Bodies/M_MED_ScotchPeek/Meshes/M_MED_ScotchPeek_AnimBP
target_skeleton=&/M_MED_ScotchPeek_Skeleton
graphs [
 AnimGraph {nodes [
   n1 = RigidBody!(ComponentPose=Local To Component(LocalPose=Use cached pose 'ControlRig'()), OverrideWorldGravity:Vector="0,0,-980", ExternalForce) -> (Pose) {
    alphaInputType=Bool
    bClampLinearTranslationLimitToRefPose=true
    bOverrideWorldGravity=true
    overridePhysicsAsset=&/M_MED_ScotchPeek_Physics
    simulationSpace=BaseBoneSpace
    alphaBoolBlend {blendOption=Linear}
    alphaScaleBiasClamp {bMapRange=true}
    baseBoneRef {boneName=pelvis}
    componentAppliedLinearAccClamp=<3000 3000 3000>
    overrideWorldGravity {z=-980}
    simSpaceSettings {
     dampingAlpha=0
     maxAngularAcceleration=40
     maxAngularVelocity=8
     maxLinearAcceleration=3500
     maxLinearVelocity=700
     velocityScaleZ=0.5
     worldAlpha=1
     externalLinearDragV=<3 3 3>
    }
   }
   Output Pose(Result=Blend Poses by bool(BlendPose_0=Use cached pose 'ControlRig'(), BlendPose_1=Component To Local(ComponentPose=@n1), bActiveValue=Get bClothEnabled()) {blendType=Linear}) {layerGroup=None}
   n2 = Control Rig!(
    Source=Copy Pose From Mesh() {bUseAttachedParent=true}
    Alpha
    BaseballSliding:bool=Get bIsBaseballSliding()
    Sliding:bool=Get bIsKneeSliding()
    isCrouching:bool=Get bIsCrouching()
    isDNBO:bool=Get bIsDBNO()
    isGrindRail:bool=Get bIsOnGrindRail()
    isRunning:bool=Get isRunning() -> (isRunning:bool)
    isKneeSliding:double="To Float (Boolean)"(InBool=Get bIsKneeSliding())
    Skydiving:bool=Get bIsSkyDiving()
    FrontEnd:bool=Get bIsInFrontEnd()
   ) -> (Pose) {controlRigAssetReference {blueprintRigClass=&/M_MED_ScotchPeek_CtrlRig.M_MED_ScotchPeek_CtrlRig_C}}
   n3 = Anim Dynamics - Bone: dyn_sleeveTop_l!(ComponentPose=Local To Component(LocalPose=@n2), Alpha=1, GravityOverride:Vector="0,0,-980", ExternalForce="0,0,0") -> (Pose) {
    angularDampingOverride=0.9999
    angularSpringConstant=50
    bAngularSpring=true
    bOverrideAngularDamping=true
    bUseGravityOverride=true
    gravityScale=0.35
    simulationSpace=RootRelative
    alphaBoolBlend {blendOption=Linear}
    boundBone {boneName=dyn_sleeveTop_l}
    componentLinearAccScale=<0.4 0.4 0.4>
    componentLinearVelScale=<0.4 0.4 0.4>
    gravityOverride {z=-980}
    physicsBodyDefinitions [{
      boundBone {boneName=dyn_sleeveTop_l}
      boxExtents=<14 14 14>
      constraintSetup {angularLimitsMax=<1 5 5>, angularLimitsMin=<-1 -5 -3>, angularTarget=<1 0 0>}
      localJointOffset=<2 0 0>
     }]
   }
   n4 = Anim Dynamics - Bone: dyn_sleeveBottom_l!(ComponentPose=@n3, Alpha=1, GravityOverride:Vector="0,0,-980", ExternalForce="0,0,0") -> (Pose) {
    angularDampingOverride=0.9999
    angularSpringConstant=50
    bAngularSpring=true
    bOverrideAngularDamping=true
    bUseGravityOverride=true
    gravityScale=0.35
    simulationSpace=RootRelative
    alphaBoolBlend {blendOption=Linear}
    boundBone {boneName=dyn_sleeveBottom_l}
    componentLinearAccScale=<0.4 0.4 0.4>
    componentLinearVelScale=<0.4 0.4 0.4>
    gravityOverride {z=-980}
    physicsBodyDefinitions [{
      boundBone {boneName=dyn_sleeveBottom_l}
      boxExtents=<14 14 14>
      constraintSetup {angularLimitsMax=<1 5 0>, angularLimitsMin=<-1 -5 -5>, angularTarget=<1 0 0>}
      localJointOffset=<2 0 0>
     }]
   }
   n5 = Anim Dynamics - Bone: dyn_sleeveBack_l!(ComponentPose=@n4, Alpha=1, GravityOverride:Vector="0,0,-980", ExternalForce="0,0,0") -> (Pose) {
    angularDampingOverride=0.9999
    angularSpringConstant=50
    bAngularSpring=true
    bOverrideAngularDamping=true
    bUseGravityOverride=true
    gravityScale=0.35
    simulationSpace=RootRelative
    alphaBoolBlend {blendOption=Linear}
    boundBone {boneName=dyn_sleeveBack_l}
    componentLinearAccScale=<0.4 0.4 0.4>
    componentLinearVelScale=<0.4 0.4 0.4>
    gravityOverride {z=-980}
    physicsBodyDefinitions [{
      boundBone {boneName=dyn_sleeveBack_l}
      boxExtents=<14 14 14>
      constraintSetup {angularLimitsMax=<1 5 3>, angularLimitsMin=<-1 -5 -3>, angularTarget=<1 0 0>}
      localJointOffset=<2 0 0>
     }]
   }
   n6 = Anim Dynamics - Bone: dyn_sleeveFront_l!(ComponentPose=@n5, Alpha=1, GravityOverride:Vector="0,0,-980", ExternalForce="0,0,0") -> (Pose) {
    angularDampingOverride=0.9999
    angularSpringConstant=50
    bAngularSpring=true
    bOverrideAngularDamping=true
    bUseGravityOverride=true
    gravityScale=0.35
    simulationSpace=RootRelative
    alphaBoolBlend {blendOption=Linear}
    boundBone {boneName=dyn_sleeveFront_l}
    componentLinearAccScale=<0.4 0.4 0.4>
    componentLinearVelScale=<0.4 0.4 0.4>
    gravityOverride {z=-980}
    physicsBodyDefinitions [{
      boundBone {boneName=dyn_sleeveFront_l}
      boxExtents=<14 14 14>
      constraintSetup {angularLimitsMax=<1 5 3>, angularLimitsMin=<-1 -5 -3>, angularTarget=<1 0 0>}
      localJointOffset=<2 0 0>
     }]
   }
   n7 = Anim Dynamics - Bone: dyn_sleeveTop_r!(ComponentPose=@n6, Alpha=1, GravityOverride:Vector="0,0,-980", ExternalForce="0,0,0") -> (Pose) {
    angularDampingOverride=0.9999
    angularSpringConstant=50
    bAngularSpring=true
    bOverrideAngularDamping=true
    bUseGravityOverride=true
    gravityScale=0.35
    simulationSpace=RootRelative
    alphaBoolBlend {blendOption=Linear}
    boundBone {boneName=dyn_sleeveTop_r}
    componentLinearAccScale=<0.4 0.4 0.4>
    componentLinearVelScale=<0.4 0.4 0.4>
    gravityOverride {z=-980}
    physicsBodyDefinitions [{
      boundBone {boneName=dyn_sleeveTop_r}
      boxExtents=<14 14 14>
      constraintSetup {angularLimitsMax=<1 3 5>, angularLimitsMin=<-1 -3 -3>, angularTarget=<1 0 0>}
      localJointOffset=<-2 0 0>
     }]
   }
   n8 = Anim Dynamics - Bone: dyn_sleeveBottom_r!(ComponentPose=@n7, Alpha=1, GravityOverride:Vector="0,0,-980", ExternalForce="0,0,0") -> (Pose) {
    angularDampingOverride=0.9999
    angularSpringConstant=50
    bAngularSpring=true
    bOverrideAngularDamping=true
    bUseGravityOverride=true
    gravityScale=0.35
    simulationSpace=RootRelative
    alphaBoolBlend {blendOption=Linear}
    boundBone {boneName=dyn_sleeveBottom_r}
    componentLinearAccScale=<0.4 0.4 0.4>
    componentLinearVelScale=<0.4 0.4 0.4>
    gravityOverride {z=-980}
    physicsBodyDefinitions [{
      boundBone {boneName=dyn_sleeveBottom_r}
      boxExtents=<14 14 14>
      constraintSetup {angularLimitsMax=<1 0 0>, angularLimitsMin=<-1 -5 0>, angularTarget=<1 0 0>}
      localJointOffset=<-2 2 0>
     }]
   }
   n9 = Anim Dynamics - Bone: dyn_sleeveBack_r!(ComponentPose=@n8, Alpha=1, GravityOverride:Vector="0,0,-980", ExternalForce="0,0,0") -> (Pose) {
    angularDampingOverride=0.9999
    angularSpringConstant=50
    bAngularSpring=true
    bOverrideAngularDamping=true
    bUseGravityOverride=true
    gravityScale=0.35
    simulationSpace=RootRelative
    alphaBoolBlend {blendOption=Linear}
    boundBone {boneName=dyn_sleeveBack_r}
    componentLinearAccScale=<0.4 0.4 0.4>
    componentLinearVelScale=<0.4 0.4 0.4>
    gravityOverride {z=-980}
    physicsBodyDefinitions [{
      boundBone {boneName=dyn_sleeveBack_r}
      boxExtents=<14 14 14>
      constraintSetup {angularLimitsMax=<1 5 3>, angularLimitsMin=<-1 -5 -3>, angularTarget=<1 0 0>}
      localJointOffset=<-2 0 0>
     }]
   }
   n10 = Anim Dynamics - Bone: dyn_sleeveFront_r!(ComponentPose=@n9, Alpha=1, GravityOverride:Vector="0,0,-980", ExternalForce="0,0,0") -> (Pose) {
    angularDampingOverride=0.9999
    angularSpringConstant=50
    bAngularSpring=true
    bOverrideAngularDamping=true
    bUseGravityOverride=true
    gravityScale=0.35
    simulationSpace=RootRelative
    alphaBoolBlend {blendOption=Linear}
    boundBone {boneName=dyn_sleeveFront_r}
    componentLinearAccScale=<0.4 0.4 0.4>
    componentLinearVelScale=<0.4 0.4 0.4>
    gravityOverride {z=-980}
    physicsBodyDefinitions [{
      boundBone {boneName=dyn_sleeveFront_r}
      boxExtents=<14 14 14>
      constraintSetup {angularLimitsMax=<1 5 3>, angularLimitsMin=<-1 -5 -3>, angularTarget=<1 0 0>}
      localJointOffset=<-2 0 0>
     }]
   }
   n11 = Anim Dynamics - Bone: dyn_sword_strings_1!(ComponentPose=@n10, Alpha=1, GravityOverride:Vector="0,0,-980", ExternalForce="0,0,0") -> (Pose) {
    angularDampingOverride=1
    angularSpringConstant=50
    bAngularSpring=true
    bOverrideAngularDamping=true
    bUseGravityOverride=true
    gravityScale=0.35
    simulationSpace=RootRelative
    boundBone {boneName=dyn_sword_strings_1}
    componentLinearAccScale=<0.4 0.4 0.4>
    componentLinearVelScale=<0.4 0.4 0.4>
    gravityOverride {z=-980}
    physicsBodyDefinitions [{
      boundBone {boneName=dyn_sword_strings_1}
      boxExtents=<15 15 15>
      constraintSetup {angularLimitsMax=<10 30 45>, angularLimitsMin=<-10 -30 -50>, angularTarget=<1 0 0>}
      localJointOffset=<4 0 0>
     }]
    simSpaceSettings {maxAngularAcceleration=24, maxAngularVelocity=6, simSpaceAngularAlpha=1}
   }
   n12 = "Transform (Modify) Bone - Bone: ground_collision"!(ComponentPose=@n11, bAlphaBoolEnabled:bool=True, Translation="0,0,0", Rotation="0,0,0", Scale="1,1,1") -> (Pose) {
    alphaInputType=Bool
    rotationMode=BMM_Replace
    rotationSpace=BCS_WorldSpace
    translationMode=BMM_Replace
    translationSpace=BCS_BoneSpace
    alphaBoolBlend {blendOption=Linear}
    boneToModify {boneName=ground_collision}
   }
   n13 = "Transform (Modify) Bone - Bone: ground_collision"(ComponentPose=@n12, bAlphaBoolEnabled=True, Translation="0,0,0", Rotation="0,0,0", Scale="1,1,1") {
    alphaInputType=Bool
    rotationSpace=BCS_WorldSpace
    translationMode=BMM_Additive
    translationSpace=BCS_BoneSpace
    alphaBoolBlend {blendInTime=0.1, blendOption=Linear, blendOutTime=0.1}
    boneToModify {boneName=ground_collision}
   }
   Save cached pose 'ControlRig'(Pose=Component To Local(ComponentPose=@n13)) {cacheName=ControlRig}
  ]}
 EventGraph {nodes [
   n1 = Event BlueprintUpdateAnimation()
   > Set delta(delta:double=@n1.DeltaTimeX) -> (Output_Get:double)
   > n2 = Sequence!(execute) -> (then_0, then_1, then_2:Exec)
   > n3 = Branch(Condition=Get bClothEnabled())
   > n4 = Sequence!(execute) -> (then_0, then_1)
   > n5 = "Cloth Controller Function (Parameter Asset) (Nativized)"!(
    execute
    CharacterPartAnimInstance:Custom Character Part Anim Instance Object Reference=Self-Reference()
    ClothingAssetName:Name=M_MED_NeonFriction_Coat_Clothing
    ClothCoreSettings:"Cloth Core Settings Structure (by ref)"=Get Cloth Core Settings() -> (Cloth Core Settings:Cloth Core Settings Structure)
    WindGustLocalSettings:"Wind Gust Settings Structure (by ref)"=Get Wind Gust Local Settings() -> (Wind Gust Local Settings:Wind Gust Settings Structure)
    ClothUpAxis:Vector="1,0,0"
    LinVelScaleFwdBwdClamp:float=0
    LinVelScaleUpDwnClamp:float=0.7
    DeltaTime:float=Get delta() -> (delta:double)
    GameTimeInSeconds:float=GetGameTimeInSeconds()
    StateLogicLayerName:Name=none
    Parameters:Fort Character Dynamics Parameters Object Reference=&/F_MED_ScotchPeek_Parameters
   ) -> (
    then
    OutCurrentState:Name=None
    OutClothParameters:Cloth Parameters Structure
    OutClothCoreSettings:Cloth Core Settings Structure
    OutWindGustSettings:Wind Gust Parameters Structure
    OutWindGustLocalSettings:Wind Gust Settings Structure
    OutWindVelocity:Vector
    OutLinearVelVector:Vector
   )
   n6 = "Gravity Override State Function (Parameter Asset) (Nativized)"(
    execute=[@n4.then_1, @n3.else]
    CharacterPartAnimInstance=Self-Reference()
    GameTimeInSeconds=GetGameTimeInSeconds()
    Parameters=&F_MED_ScotchPeek_Parameters
   )
   > Set Gravity Override(Gravity Override=@n6)
   n7 = Get bIsCrouchSprinting()
   n8 = Get bIsTacticalSprinting()
   n9 = Get bIsSprinting()
   Set isRunning(execute=@n2.then_1, isRunning:bool=AND Boolean(A=OR Boolean(A=@n8, B=@n9), B=NOT Boolean(A=@n7))) -> (Output_Get:bool)
   Handle Slope Warp(execute=@n2.then_2)
   PrintString!(execute, InString="To String (Name)"(InName=@n5), bPrintToScreen, bPrintToLog, TextColor, Duration, Key) -> (then)
  ]}
 Handle Slope Warp {nodes [
   Handle Slope Warp() {is=title K2Node_FunctionEntry}
   > n1 = Sequence()
   n2 = Get bIsOnGround() -> (bIsOnGround:bool)
   > n3 = Branch(Condition=OR Boolean!(A=NOT Boolean(A=@n2), B=Get bIsPlayingEmote(), C:bool=Get bIsInVehicle() -> (bIsInVehicle:bool)) -> (ReturnValue))
   > Set Target Info Input Root Slope Translation(Target Info Input Root Slope Translation:Vector="0,0,-100") -> (Output_Get:Vector)
   n4 = Get delta()
   Set Slope Info Input Slope Rotation(
    execute=@n3.else
    Slope Info Input Slope Rotation:Rotator=RInterpTo!(
     Current:Rotator=Get Slope Info Input Slope Rotation() -> (Slope Info Input Slope Rotation:Rotator)
     Target:Rotator=Get!() -> (
      SlopeInfoInput_RootSlopeTranslation:Vector=~
      SlopeInfoInput_LeftFootSlopeTranslation:Vector=~
      SlopeInfoInput_RightFootSlopeTranslation:Vector=~
      SlopeInfoInput_SlopeRotation:Rotator=~
     ).SlopeInfoInput_SlopeRotation
     DeltaTime:float=@n4
     InterpSpeed:float=7
    ) -> (ReturnValue:Rotator)
   ) -> (Output_Get:Rotator)
   > Set Target Info Input Root Slope Translation(Target Info Input Root Slope Translation="0,0,5")
   Set Slope Info Input Root Slope Translation(
    execute=@n1.then_1
    Slope Info Input Root Slope Translation:Vector=VInterpTo!(
     Current:Vector=Get Slope Info Input Root Slope Translation() -> (Slope Info Input Root Slope Translation:Vector)
     Target:Vector=Get Target Info Input Root Slope Translation() -> (Target Info Input Root Slope Translation:Vector)
     DeltaTime:float=@n4
     InterpSpeed:float=10
    ) -> (ReturnValue:Vector)
   ) -> (Output_Get:Vector)
  ]}
]
variables [
 Cloth Core Settings {
  type=ClothCoreSettings
  default {
   collisionThickness=0
   dampingCoefficient=0.05
   frictionCoefficient=0.3
   localDampingCoefficient=0
   numberIterations=2
   numberSubsteps=1
   selfCollisionThickness=0
   animDriveDamping=<0.2 1>
   animDriveStiffness=<0.2 1>
   areaStiffness=<1 1>
   bendingStiffness=<1 1>
   edgeStiffness=<1 1>
   tetherStiffness=<1 1>
  }
 }
 delta {type=number}
 Gravity Override {type=Vector}
 isRunning {type=boolean}
 Slope Info Input Root Slope Translation {type=Vector}
 Slope Info Input Slope Rotation {type=Rotator}
 Target Info Input Root Slope Translation {type=Vector}
 Wind Gust Local Settings {type=WindGustSettings, default {elapsedTime=0, finalGustStrengthLocal=0, gustDuration=0, gustStrength=0, timeBetweenGusts=0}}
]
