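# Cloth collision Physics Asset for the ScotchPeek medium male player cosmetic body, judging by its name and path.
# Capsules cover the pelvis, spine and dyn_Sword bone, with a tiny one on root; arm, leg and foot bodies list no shapes.
# A large box on ground_collision forms a floor with its top near bone height.
# A few constraints link sword, forearms and upper spine to spine bones with free swing and twist.
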PhysicsAsset /BRCosmetics/Characters/Player/Male/Medium/Bodies/M_MED_ScotchPeek/Meshes/M_MED_ScotchPeek_ClothCollision
bodies {
 calf_l []
 calf_r []
 dyn_Sword [
  {
   shapeType=Capsule
   length=113.3012
   radius=3.8864
   shapeName=dyn_Sword_capsule
   center=<16.3675 9.6033 0.1538>
   rotation=<rot -0.0121 114.7666 -89.5323>
  }
  {
   shapeType=Capsule
   length=59.359
   radius=7.1843
   shapeName=dyn_Sword_capsule
   center=<37.8872 19.1908 1.2229>
   rotation=<rot 87.1117 -101.6486 -129.7267>
  }
 ]
 foot_l []
 foot_r []
 ground_collision [{shapeType=Box, extentX=331.4972, extentY=332.5133, extentZ=73.6069, shapeName=root_box, center=<0 0 -36.7803>, rotation=<rot 0 -90 0>}]
 lowerarm_l []
 lowerarm_r []
 pelvis [
  {shapeType=Capsule, length=7.3536, radius=14.313, center=<-2.4887 -0.6978 0>}
  {shapeType=Capsule, length=20.426, radius=15.9555, shapeName=pelvis_capsule, center=<-29.896 0.2054 0>, rotation=<rot 0 89.0607 90>}
 ]
 root [{shapeType=Capsule, length=0.505, radius=0.505, shapeName=root_capsule, rotation=<rot 0 -90 0>}]
 spine_01 [{shapeType=Capsule, length=10.785, radius=10.1267, center=<0.3095 -0.6696 0>}]
 spine_02 [{shapeType=Capsule, length=6.9335, radius=13.6614, center=<3.597 0 0>}]
 spine_03 [{shapeType=Capsule, length=5.6227, radius=13.1876, shapeName=spine_03_capsule, center=<7.3745 0 0>}]
 thigh_l []
 thigh_r []
}
constraints [
 {
  bone1Name=dyn_Sword
  bone2Name=spine_01
  swing1LimitDegrees=45
  swing1Motion=Free
  swing2LimitDegrees=45
  swing2Motion=Free
  twistLimitDegrees=45
  twistMotion=Free
 }
 {
  bone1Name=lowerarm_l
  bone2Name=spine_03
  swing1LimitDegrees=45
  swing1Motion=Free
  swing2LimitDegrees=45
  swing2Motion=Free
  twistLimitDegrees=45
  twistMotion=Free
 }
 {
  bone1Name=lowerarm_r
  bone2Name=spine_03
  swing1LimitDegrees=45
  swing1Motion=Free
  swing2LimitDegrees=45
  swing2Motion=Free
  twistLimitDegrees=45
  twistMotion=Free
 }
 {
  bone1Name=spine_03
  bone2Name=spine_02
  swing1LimitDegrees=45
  swing1Motion=Free
  swing2LimitDegrees=45
  swing2Motion=Free
  twistLimitDegrees=45
  twistMotion=Free
 }
]
