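# Control Rig for the ScotchPeek character's face accessory part that corrects hair bones, including the bottom left and right bases, as the head moves.
# Spherical pose readers measure head orientation against spine_05 and weight Modify Transforms offsets.
# A branch offers two reader and offset setups, but its condition is fixed to true, so only the symmetric-region setup runs.
# The IsTargeting variable node is left unconnected.
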ControlRigBlueprint /BRCosmetics/Characters/Player/Male/Medium/Bodies/M_MED_ScotchPeek/Meshes/Parts/M_MED_ScotchPeek_FaceAcc_CtrlRig
graphs [
 RigVMFunctionLibrary {links [], nodes []}
 RigVMModel {nodes [
   RigUnit_BeginExecution = &RigUnit_BeginExecution()
   n1 = &RigVMFunction_ControlFlowBranch(Condition=true, ExecuteContext=@RigUnit_BeginExecution.ExecutePin)
   n2 = &RigUnit_SphericalPoseReader(
    ActiveRegionScaleFactors=(PositiveWidth=1.3,NegativeWidth=0.5,PositiveHeight=0.5,NegativeHeight=0.3)
    ActiveRegionSize=0.08
    Debug=(bDrawDebug=true,bDraw2D=False,bDrawLocalAxes=false,DebugScale=25,DebugSegments=20,DebugThickness=0.25)
    DriverAxis=(X=1,Y=0,Z=0)
    DriverItem=(Type=Bone,Name="head")
    FalloffRegionScaleFactors=(PositiveWidth=0.2,NegativeWidth=0.2,PositiveHeight=0.2,NegativeHeight=0.25)
    FalloffSize=0.5
    FlipHeightScaling=False
    FlipWidthScaling=False
    OptionalParentItem=(Type=Bone,Name="spine_05")
    RotationOffset=(X=-53.8714,Y=70.1264,Z=80.6906)
    ExecutePin=@n1.False
   )
   > &RigUnit_ModifyTransforms(
    ItemToModify="((Item=(Type=Bone,Name=\"Base_left_bottom_hair\"),Transform=(Rotation=(X=-0.0739,Y=0.2231,Z=-0.6631,W=0.7106),Translation=(X=-3.0106,Y=-1.2671,Z=1.1332),Scale3D=(X=1,Y=1,Z=1))),(Item=(Type=Bone,Name=\"Base_right_bottom_hair\"),Transform=(Rotation=(X=0,Y=-0.316 ...(truncated)"
    WeightMinimum=0
    Weight=@n2.OutputParam
   )
   n3 = &RigUnit_SphericalPoseReader(
    ActiveRegionScaleFactors=(PositiveWidth=0.8,NegativeWidth=0.8,PositiveHeight=0.8,NegativeHeight=0.8)
    ActiveRegionSize=0.08
    Debug=(bDrawDebug=true,bDraw2D=False,bDrawLocalAxes=False,DebugScale=25,DebugSegments=20,DebugThickness=0.25)
    DriverAxis=(X=1,Y=0,Z=0)
    DriverItem=(Type=Bone,Name="head")
    FalloffRegionScaleFactors=(PositiveWidth=0.7,NegativeWidth=0.7,PositiveHeight=0.7,NegativeHeight=0.7)
    FalloffSize=0.5
    FlipHeightScaling=False
    FlipWidthScaling=False
    OptionalParentItem=(Type=Bone,Name="spine_05")
    RotationOffset=(X=-45,Y=0,Z=40)
    ExecutePin=@n1.True
   )
   > &RigUnit_ModifyTransforms(
    ItemToModify="((Item=(Type=Bone,Name=\"Base_left_bottom_hair\"),Transform=(Rotation=(X=-0.1496,Y=0.1812,Z=-0.8753,W=0.4226),Translation=(X=-4.0495,Y=0.032,Z=1.594),Scale3D=(X=1,Y=1,Z=1))),(Item=(Type=Bone,Name=\"Base_right_bottom_hair\"),Transform=(Rotation=(X=0,Y=-0.3167,Z ...(truncated)"
    WeightMinimum=0
    Weight=@n3.OutputParam
   )
   RigVMVariableNode() {variable="IsTargeting "}
  ]}
]
hierarchy [
 root {loc=<0 0 0>, rot=<0 0 0 1>}
 pelvis {parent=root, loc=<0 -0.7071 93.1191>, rot=<0.0127 0.707 -0.0127 -0.707>}
 spine_01 {parent=pelvis, loc=<6.7494 -0.6388 0>, rot=<0 0 0.034 -0.9994>}
 spine_02 {parent=spine_01, loc=<7.2518 -0.035 0>, rot=<0 0 0.0238 -0.9997>}
 spine_03 {parent=spine_02, loc=<6.0104 0 0>, rot=<0 0 0.0149 -0.9999>}
 spine_04 {parent=spine_03, loc=<8.1024 0 0>, rot=<0 0 -0.0132 -0.9999>}
 spine_05 {parent=spine_04, loc=<10.0146 0 0>, rot=<0 0 -0.0015 -1>}
 neck_01 {parent=spine_05, loc=<11.3138 0.1529 0>, rot=<0 0 -0.2003 -0.9797>}
 neck_02 {parent=neck_01, loc=<4.5627 0 0>, rot=<0 0 0.0504 -0.9987>}
 head {parent=neck_02, loc=<5.3914 0 0>, rot=<0 0 0.115 -0.9934>}
 Base_front_hair {parent=head, loc=<14.0179 13.4963 1.1255>, rot=<0.0402 -0.9974 0.0019 -0.0603>}
 dyn_front_hair_1 {parent=Base_front_hair, loc=<0 0 0>, rot=<0 0 0 1>}
 dyn_front_hair_2 {parent=dyn_front_hair_1, loc=<4.8435 0 0>, rot=<0 0 0 1>}
 Base_middle_hair {parent=head, loc=<4.3395 -11.9454 -0.9409>, rot=<0.214 -0.6917 0.6513 -0.2269>}
 dyn_middle_hair_1 {parent=Base_middle_hair, loc=<0 0 0>, rot=<0 0 0 1>}
 Base_middle_bottom_hair {parent=head, loc=<0.1519 -11.669 -0.6982>, rot=<0.15 -0.9887 0 0>}
 dyn_middle_bottom_hair_1 {parent=Base_middle_bottom_hair, loc=<0 0 0>, rot=<0 0 0 1>}
 Base_right_top_hair {parent=head, loc=<12.3805 -12.8578 3.6149>, rot=<0.2457 -0.597 0.7331 -0.214>}
 dyn_right_top_hair {parent=Base_right_top_hair, loc=<0 0 0>, rot=<0 0 0 1>}
 Base_left_middle_hair {parent=head, loc=<3.1251 -7.204 -8.2711>, rot=<0.1474 -0.9428 0.2957 0.0437>}
 dyn_left_middle_hair_1 {parent=Base_left_middle_hair, loc=<0 0 0>, rot=<0 0 0 1>}
 Base_right_middle_hair {parent=head, loc=<3.333 -9.0691 6.279>, rot=<-0.0528 -0.2694 0.9421 -0.1928>}
 dyn_right_middle_hair {parent=Base_right_middle_hair, loc=<0 0 0>, rot=<0 0 0 1>}
 Base_left_top_hair {parent=head, loc=<13.5442 -9.3724 -8.5804>, rot=<0.2219 -0.8592 0.4482 -0.1082>}
 dyn_left_top_hair {parent=Base_left_top_hair, loc=<0 0 0>, rot=<0 0 0 1>}
 Base_right_side_hair {parent=head, loc=<7.3339 5.0421 10.2133>, rot=<0.028 0.282 0.9535 -0.1027>}
 dyn_right_side_hair {parent=Base_right_side_hair, loc=<0 0 0>, rot=<0 0 0 1>}
 Base_left_side_hair {parent=head, loc=<8.9753 4.7349 -10.4733>, rot=<-0.1098 -0.9903 -0.0829 -0.0194>}
 dyn_left_side_hair {parent=Base_left_side_hair, loc=<0 0 0>, rot=<0 0 0 1>}
 Base_left_bottom_hair {parent=head, loc=<-0.7254 -8.3289 5.5836>, rot=<0.035 -0.9449 0.3142 -0.085>}
 dyn_left_bottom_hair_1 {parent=Base_left_bottom_hair, loc=<0 0 0>, rot=<0 0 0.133 -0.9911>}
 Base_right_bottom_hair {parent=head, loc=<-0.7254 -8.3289 -5.5836>, rot=<0.035 -0.9449 0.3142 -0.085>}
 dyn_left_bottom_hair_11 {parent=Base_right_bottom_hair, loc=<0 0 0>, rot=<0 0 0.133 -0.9911>}
]
variables ["IsTargeting " {type=bool, default=False}]
